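# Translucent unlit material for Niagara sprite and mesh particles in the DoubleConk backpack's inactive respawn effect, drawing an animated tiled grid panel.
# Two panning noise samples, randomized flickering tiles, sine-derived grid lines and a parameter-driven intro swipe supply layers for a two-color gradient with color dodge and tile darkening.
# Eased UV sine curves form the procedural shape mask, with adjustable edge dimming.
# Particle color and alpha scale the result.
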Material /BRCosmetics/Accessories/FORT_Backpacks/Backpack_M_MED_DoubleConk/FX/M_Backpack_DoubleConk_InactiveRespawn
graph {
 properties {
  bAutomaticallySetUsageInEditor=false
  bUsedWithNiagaraMeshParticles=true
  bUsedWithNiagaraSprites=true
  blendMode=BLEND_Translucent
  pixelDepthOffsetMode=PDOM_Legacy
  shadingModel=MSM_Unlit
  usedShadingModels=Unlit
 }
 nodes [
  AppendVector(A=Constant() {r=0.04}, B=Constant() {r=0.04})
  n1 = ParticleColor()
  n2 = LinearInterpolate(A=NamedRerouteUsage(), Alpha=NamedRerouteUsage())
  n3 = Multiply(A=@n2, B=NamedRerouteUsage())
  n4 = MaterialFunctionCall!(
   Base=LinearInterpolate(
    A=VectorParameter() {parameterName=Color_A, defaultValue=<rgba 0 0.05 1 1>}.RGB
    B=VectorParameter() {parameterName=Color_B, defaultValue=<rgba 0 0.25 1 1>}.RGB
    Alpha=LinearInterpolate(A=@n3, B=@n2)
   )
   Add=Multiply(A=NamedRerouteUsage(), B=ScalarParameter() {defaultValue=0.5, parameterName=GridLinesAlpha})
  ) {materialFunction=/Engine/Functions/Engine_MaterialFunctions02/Math/MultiplyAdd}
  n5 = Saturate(None=MaterialFunctionCall!(Blend=NamedRerouteUsage(), Base=Saturate(None=@n4.Result)) {materialFunction=/Engine/Functions/Engine_MaterialFunctions03/Blends/Blend_ColorDodge}.Result)
  n6 = ScalarParameter() {defaultValue=0.5, parameterName=TileSubtraction_Amount}
  n7 = NamedRerouteUsage()
  n8 = StaticSwitchParameter(
   True=NamedRerouteUsage()
   False=Saturate(None=MaterialFunctionCall("Value (0-1)"=@n7) {materialFunction=&MF_UI_EaseCurves}.EaseOutCubic)
  ) {parameterName=RoundedFadeMask}
  n9 = LinearInterpolate(A=@n5, B=Multiply(A=@n5, B=OneMinus(None=@n6)) {constB=0.666}, Alpha=Min(A=NamedRerouteUsage(), B=OneMinus(None=@n8))) {constB=0}
  n10 = OneMinus(None=ScalarParameter() {parameterName=DimEdgeAmount})
  n11 = Sine(None=TextureCoordinate()) {period=2}
  n12 = MaterialFunctionCall("Value (0-1)"=ComponentMask(None=@n11) {g=true}) {materialFunction=&MF_UI_EaseCurves}
  n13 = ComponentMask(None=@n11) {r=true}
  n14 = MaterialFunctionCall("Value (0-1)"=@n13) {materialFunction=&MF_UI_EaseCurves}
  n15 = Saturate(None=MaterialFunctionCall("Value (0-1)"=@n14.EaseInOutSine) {materialFunction=&MF_UI_EaseCurves}.EaseInOutQuint)
  n16 = LinearInterpolate(A=@n15, B=Saturate(None=@n14.EaseInQuint), Alpha=ScalarParameter() {parameterName=FadeSelect})
  n17 = LinearInterpolate(A=@n16, B=LinearInterpolate(B=@n16, Alpha=NamedRerouteUsage()), Alpha=ScalarParameter() {parameterName=DimEdgeAmount})
  n18 = Multiply(A=@n12.EaseInQuint, B=@n17)
  n19 = Min(A=@n12.EaseInQuint, B=@n17)
  n20 = ScalarParameter() {defaultValue=0.2, parameterName=ShapeChangeRatio}
  n21 = NamedRerouteDeclaration(None=Saturate(None=LinearInterpolate(A=@n18, B=@n19, Alpha=@n20) {constAlpha=0.2})) {name=ShapeMask, nodeColor=<rgba 0 0 0 1>}
  n22 = LinearInterpolate(B=LinearInterpolate(A=Multiply(A=@n9, B=@n10), B=@n9, Alpha=NamedRerouteUsage()), Alpha=@n21)
  MaterialFunctionCall(Emissive=Multiply(A=@n1.RGB, B=Saturate(None=@n22))) {materialFunction=&MF_FX_EmissiveAdjustments}
  n23 = TextureCoordinate()
  n24 = Sine(None=@n23) {period=2}
  n25 = ComponentMask(None=OneMinus(None=Add(A=@n24) {constB=0})) {g=true}
  MaterialFunctionCall(R=Power(Base=ComponentMask(None=@n24) {r=true}, Exp=Multiply(A=@n25) {constB=2}) {constExponent=1}, G=ComponentMask(None=@n24) {g=true}) {materialFunction=&AppendMany}
  Multiply(A=NamedRerouteUsage(), B=@n12.EaseInQuint)
  n26 = Multiply(A=ScalarParameter() {defaultValue=20, parameterName=GridSize})
  n27 = Divide(B=@n26)
  Multiply(A=@n27, B=@n27)
  n28 = Constant()
  n29 = MaterialFunctionCall!(
   UV=NamedRerouteUsage()
   Global Offset
   Global Rotation
   Tiling=ScalarParameter() {defaultValue=20, parameterName=GridSize}
   Cell Offset
   Cell Scale=@n28
   Cell Rotation=@n28
   Random Seed
  ) {materialFunction=/FortUILibrary/MaterialFunctions/Basics/MF_UI_ParticleGrid}
  n30 = Time()
  n31 = ScalarParameter() {defaultValue=0.2, parameterName=TileSpeed}
  n32 = Saturate(None=MaterialFunctionCall!(UV=@n29.Random Cell Values, Seed=Multiply(A=@n30, B=@n31)) {materialFunction=&MF_UI_Random}.Result)
  NamedRerouteDeclaration(None=Multiply(A=SmoothStep(Value=@n32) {constMin=0.6}, B=ScalarParameter() {defaultValue=1, parameterName=AmbientTile_Intensity}) {constB=0.2}) {name=Tiles, nodeColor=<rgba 0 0 0 1>}
  n33 = NamedRerouteUsage()
  n34 = Sine(None=Multiply(A=ConstantBiasScale(None=@n33) {bias=2}, B=@n26) {constB=10})
  n35 = MaterialFunctionCall(Float2=@n34) {materialFunction=&BreakOutFloat2Components}
  n36 = MaterialFunctionCall("Value (0-1)"=MaterialFunctionCall("Value (0-1)"=@n35.R) {materialFunction=&MF_UI_EaseCurves}.EaseInQuint) {materialFunction=&MF_UI_EaseCurves}
  n37 = MaterialFunctionCall("Value (0-1)"=MaterialFunctionCall("Value (0-1)"=@n35.G) {materialFunction=&MF_UI_EaseCurves}.EaseInQuint) {materialFunction=&MF_UI_EaseCurves}
  NamedRerouteDeclaration(None=Max(A=Saturate(None=@n36.EaseInQuint), B=Saturate(None=@n37.EaseInQuint))) {name=GridLines, nodeColor=<rgba 0 0 0 1>}
  NamedRerouteDeclaration(None=@n34) {name=ManipulatedScaledUVs, nodeColor=<rgba 0 0 0 1>}
  NamedRerouteDeclaration(None=@n15) {name=HorizontalSine, nodeColor=<rgba 0.5234 1 0 1>}
  n38 = Multiply(
   A=NamedRerouteDeclaration(None=@n23) {name=ScaledUVs, nodeColor=<rgba 0 0 0 1>}
   B=ScalarParameter() {defaultValue=0.5, parameterName=Noise1_Scale}
  )
  n39 = AppendVector(
   A=ScalarParameter() {defaultValue=-0.075, parameterName=Noise1_PanSpeedX}
   B=ScalarParameter() {defaultValue=0.075, parameterName=Noise1_PanSpeedY}
  )
  n40 = Multiply(A=NamedRerouteUsage(), B=ScalarParameter() {defaultValue=0.7, parameterName=Noise2_Scale})
  n41 = AppendVector(
   A=ScalarParameter() {defaultValue=0.05, parameterName=Noise2_PanSpeedX}
   B=ScalarParameter() {defaultValue=-0.05, parameterName=Noise2_PanSpeedY}
  )
  n42 = Multiply(
   A=TextureSample(UVs=Panner(Coordinate=@n38, Speed=@n39)) {samplerType=SAMPLERTYPE_Grayscale, texture=/FortUILibrary/UIKit/Buttons/Textures/LowResBlurredNoise}.R
   B=TextureSample(UVs=Panner(Coordinate=@n40, Speed=@n41)) {samplerType=SAMPLERTYPE_Grayscale, texture=&LowResBlurredNoise}.R
  )
  NamedRerouteDeclaration(None=Saturate(None=Multiply(A=@n42) {constB=2})) {name=Noise, nodeColor=<rgba 0 0 0 1>}
  n43 = MaterialFunctionCall(UV=@n29.Random Cell Values, Seed=Multiply(A=@n30, B=ScalarParameter() {defaultValue=1, parameterName=IntroSwipeTileSpeed})) {materialFunction=&MF_UI_Random}
  NamedRerouteDeclaration(None=LinearInterpolate(Alpha=Saturate(None=@n43.Result)) {constA=0.333}) {name=IntroSwipeTiles, nodeColor=<rgba 0 0 0 1>}
  n44 = ScalarParameter() {parameterName=Anim_Swipe}
  n45 = Sine(None=Add(A=TextureCoordinate(), B=LinearInterpolate(Alpha=@n44) {constA=1, constB=-1})) {period=2}
  n46 = LinearInterpolate(A=ComponentMask(None=@n45) {g=true}, B=ComponentMask(None=@n45) {r=true}, Alpha=ScalarParameter() {parameterName=SwipeDireciton})
  n47 = NamedRerouteDeclaration(None=Saturate(None=MaterialFunctionCall("Value (0-1)"=@n46) {materialFunction=&MF_UI_EaseCurves}.EaseInCubic)) {name=IntroSwipe, nodeColor=<rgba 0 0 0 1>}
  n48 = NamedRerouteUsage()
  NamedRerouteDeclaration(None=Multiply(A=Multiply(A=@n47, B=@n48), B=ScalarParameter() {defaultValue=1, parameterName=SwipeIntensity}) {constB=0}) {name=Anim_IntroSwipe, nodeColor=<rgba 0 0 0 1>}
  n49 = Multiply(A=NamedRerouteDeclaration(None=MaterialFunctionCall() {materialFunction=&RadialGradientExponential}.RadialGradientExponential) {name=SoftRoundMask, nodeColor=<rgba 0 0 0 1>})
  NamedRerouteDeclaration(None=Saturate(None=Multiply(A=@n49) {constB=2})) {name=FadeMask, nodeColor=<rgba 0 0 0 1>}
  NamedRerouteDeclaration(None=SmoothStep(Value=@n32) {constMin=0.9}) {name=SubtractionTiles, nodeColor=<rgba 0 0 0 1>}
  NamedRerouteUsage()
  n50 = Add(A=@n11) {constB=0}
  Power(Base=@n13, Exp=ComponentMask(None=OneMinus(None=@n50)) {g=true})
  n51 = LinearInterpolate(B=ScalarParameter() {defaultValue=0.5, parameterName=AdditiveMix}, Alpha=ScalarParameter() {defaultValue=1, parameterName=Anim_AdditiveMix}) {constA=0.3}
  Saturate(None=Multiply(A=@n1.A, B=LinearInterpolate(B=@n51, Alpha=@n21) {constB=0.25}))
  SmoothStep(Value=@n14.EaseInQuint) {constMax=0.25}
 ]
}
