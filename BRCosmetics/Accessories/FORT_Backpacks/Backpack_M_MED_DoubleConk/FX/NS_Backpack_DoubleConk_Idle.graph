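# Idle Niagara effect for the DoubleConk cosmetic backpack, with all emitters placed through a shared socket reader.
# Illumination, cone glow, GPU streaks, drifting cubes, screen lights and headlights form the look.
# User.Reactivity fades or shrinks the idle elements, while User.GlitchSpawn adds multicolored glitch particles and brightens the light.
# Glows and screens respond to camera angle, and headlight alpha follows time of day in gameplay, with front-end overrides.
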NiagaraSystem /BRCosmetics/Accessories/FORT_Backpacks/Backpack_M_MED_DoubleConk/FX/NS_Backpack_DoubleConk_Idle
dependencies {
 usedDataInterfaces [
  &NiagaraDataInterfaceColorCurve
  &NiagaraDataInterfaceCurve
  &NiagaraDataInterfaceCamera
  &NiagaraDataInterfaceSocketReader
  &NiagaraDataInterfaceArrayFloat
 ]
 usedDynamicInputs [&SetBoolByFloatComparison]
 usedModules [
  &self:SystemSpawnScript.SystemScriptSource.SystemScriptGraph.NiagaraNodeAssignment_0.SetVariables_#1
  &self:SystemSpawnScript.SystemScriptSource.SystemScriptGraph.NiagaraNodeAssignment_4.SetVariables_#2
  &SystemState
  &Timeline
  &self:SystemSpawnScript.SystemScriptSource.SystemScriptGraph.NiagaraNodeAssignment_6.SetVariables_#3
  &self:SystemSpawnScript.SystemScriptSource.SystemScriptGraph.NiagaraNodeAssignment_7.SetVariables_#4
 ]
}
emitters {
 ConeGlow {
  emitterUpdateScript [
   &EmitterState()
   &SpawnRate(SpawnRate=8, Use Spawn Probability=false)
   &SpawnRate(SpawnGroup=1, SpawnRate=&Multiply_Float(A=70, B=$User.GlitchSpawn), Use Spawn Probability=false) {module=SpawnRate001}
  ]
  particleSpawnScript [
   /SetVariables(Particles.VisibilityTag=$Emitter.SpawnGroup) {isSetParameters=true}
   &InitializeParticle(
    Color=&MergeHelper_ColorSwitch(0=<rgba 0.9174 0.8073 1 0.5>, 1=<rgba 0 5 1.3182 2>, 2=?, 3=?, Index=$Particles.VisibilityTag, Num=::NewEnumerator0)
    Lifetime=&MergeHelper_FloatSwitch(
     0=&RandomRangeFloat(Minimum=1, Maximum=1.5)
     1=0.3
     2=?
     3=?
     Index=$Particles.VisibilityTag
     Num=::NewEnumerator0
    )
    Mesh Scale Max=&MakeCustomVectorFromBool(Bool=&RandomBool(), True Vector=<2 1 1>, False Vector=<2 2 5>)
    Mesh Scale Min=<0.2 0.2 0.3>
    Mesh Scale Mode=::NewEnumerator2
    Position Offset=<0 0 -5>
    Sprite Size Max=<30 20>
    Sprite Size Min=<20 10>
    Sprite Size Mode=::NewEnumerator2
    UsePositionOffset=true
    Write Material Random=true
   )
   &SubUVAnimation(End Frame=4, SubUV Animation Mode:ENiagara_SubUVLookupMode=::NewEnumerator2)
   &SocketLocation(SamplePosOffset=false, Socket Reader=$System.Socket Reader, Socket Sampling Mode=::NewEnumerator3, VelocitySpeedLimit=false)
   &ShapeLocation(
    Cone Angle=110
    Cone Length=15
    MaskSpawnGroup=false
    Non Uniform Scale=&MergeHelper_VectorSwitch(0=<0 0 0>, 1=<1 1 1>, 2=?, 3=?, Index=$Particles.VisibilityTag, Num=::NewEnumerator0)
    Offset Mode=::NewEnumerator0
    Override Seed=false
    Shape Primitive=::NewEnumerator5
    UseOwnerScale=true
   ) {module=ShapeLocation001}
   &DynamicMaterialParameters(Module.Index0Float0Write=true, Module.Index0Float1Write=true, Module.Index0Float2Write=true, Module.Index0Float3Write=true)
  ]
  particleUpdateScript [
   &ParticleState()
   &SpriteFacingAndAlignment(Sprite Alignment=<0 0 1>, WriteAlignment=true, WriteFacing=false)
   &ScaleColor(
    Scale Alpha=&FloatFromCurve(
     FloatCurve=DataInterface() {propertyValues {Curve {keys [{interpMode=RCIM_Cubic, tangentMode=RCTM_User, time=0, value=0}, {interpMode=RCIM_Cubic, tangentMode=RCTM_User, time=0.3, value=1}, <at 1 0>]}}}
     Scale Curve=&Multiply_Float(
      A=&FN_FrontEndFloatModifier(Gameplay=1, Front End=0.9)
      B=&Lerp_Float(A=1.5, B=2, Alpha=$System.Timeline.Float)
     )
    )
    Scale RGB=&MakeVectorFromLinearColorRGB(Color=&MergeHelper_ColorSwitch(
      0=&ColorFromCurve(ColorCurve=DataInterface() {propertyValues {
         AlphaCurve {keys [<at 0 1 RCIM_Cubic>,<at 1 1 RCIM_Cubic>]}
         BlueCurve {keys [<at 0.4 1 RCIM_Cubic>,<at 0.7 1>]}
         GreenCurve {keys [<at 0.4 0.788 RCIM_Cubic>,<at 0.7 0.232>]}
         RedCurve {keys [<at 0.4 0.168 RCIM_Cubic>,<at 0.7 0.155>]}
        }})
      1=<rgba 1 1 1 1>
      2=?
      3=?
      Index=$Particles.VisibilityTag
      Num=::NewEnumerator0
     ))
    ScaleA=true
    ScaleRGB=true
   )
   &ScaleSpriteSize(
    Scale Sprite Size Mode=::NewEnumerator0
    Uniform Scale Factor=&Lerp_Float(A=0.65, B=1, Alpha=$System.Timeline.Float)
   )
   &ScaleMeshSize(Scale Factor=&Multiply_VectorByFloat(
     Vector=<1 1 1>
     Float=&Multiply_Float(
      A=0.2
      B=&FloatFromCurve(FloatCurve=DataInterface() {propertyValues {Curve {keys [
           {arriveTangent=4.1785, time=0, value=1}
           <at 0.5 1.1>
           {arriveTangent=-0.0405, interpMode=RCIM_Cubic, leaveTangent=-0.0405, tangentMode=RCTM_User, time=1, value=1.5}
          ]}}})
     )
    ))
   &CameraOffset(Camera Offset Amount=6)
   &ScaleColor(
    Scale Alpha=&MergeHelper_FloatSwitch(
     0=&Subtract_Float(
      A=&OneMinusFloat(Float=$User.Reactivity)
      B=&Abs_Float(Float:NiagaraFloat=&DotProduct(Vector A=&ReturnCameraVector(Camera Vectors=::NewEnumerator1, Coordinate Space=::Simulation), Vector B=<0 0 1>))
     )
     1=&FloatFromCurve(
      FloatCurve=DataInterface() {propertyValues {Curve {keys [
          {arriveTangent=9.0725, interpMode=RCIM_Cubic, leaveTangent=9.0725, tangentMode=RCTM_User, time=0, value=0}
          {arriveTangent=-2.5249, interpMode=RCIM_Cubic, leaveTangent=-2.5249, tangentMode=RCTM_User, time=0.2, value=0.9847}
          {arriveTangent=-0.0405, interpMode=RCIM_Cubic, leaveTangent=-0.0405, tangentMode=RCTM_User, time=1, value=0}
         ]}}}
      Scale Curve=&RandomRangeFloat(Minimum=0.2, Maximum=0.6)
     )
     2=?
     3=?
     Index=$Particles.VisibilityTag
     Num=::NewEnumerator0
    )
    Scale RGB=&MergeHelper_VectorSwitch(
     0=<1 1 1>
     1=&Multiply_VectorByFloat(
      Vector=&MakeVectorFromLinearColorRGB(Color=&ColorFromCurve(ColorCurve=DataInterface() {propertyValues {
          AlphaCurve {keys []}
          BlueCurve {keys [
            <at 0 5 RCIM_Cubic>
            <at 0.1326 5>
            <at 0.2869 5>
            <at 0.4134 0>
            <at 0.5108 5>
            <at 0.6236 0.0803>
            {arriveTangent=26.436, interpMode=RCIM_Cubic, leaveTangent=26.436, time=0.7107, value=0.8321}
            <at 0.8097 5>
           ]}
          GreenCurve {keys [
            <at 0 4.702 RCIM_Cubic>
            <at 0.1326 0>
            <at 0.2869 0.6611>
            <at 0.4134 0.0442>
            <at 0.5108 0.7585>
            <at 0.6236 5>
            {arriveTangent=-21.532, interpMode=RCIM_Cubic, leaveTangent=-21.532, time=0.7107, value=1}
            <at 0.8097 0.9929>
           ]}
          RedCurve {keys [<at 0 0 RCIM_Cubic>,<at 0.1326 2.465>,<at 0.2869 0>,<at 0.4134 5>,<at 0.5108 4.6237>,<at 0.6236 0>,<at 0.7107 0 RCIM_Cubic>,<at 0.8097 0>]}
         }}))
      Float=&FloatFromCurve(
       FloatCurve=DataInterface() {propertyValues {Curve {keys [{arriveTangent=-2.0684, interpMode=RCIM_Cubic, leaveTangent=-2.0684, tangentMode=RCTM_User, time=0, value=2}, <at 1 0>]}}}
       Scale Curve=&FN_FrontEndFloatModifier(Gameplay=1, Front End=1)
      )
     )
     2=?
     3=?
     Index=$Particles.VisibilityTag
     Num=::NewEnumerator0
    )
    ScaleA=true
    ScaleRGB=true
   ) {module=ScaleColor001}
  ]
  rendererClasses [&NiagaraSpriteRendererProperties, &NiagaraMeshRendererProperties]
  renderers [
   {rendererClass=&NiagaraSpriteRendererProperties}
   {rendererClass=&NiagaraMeshRendererProperties, rendererIndex=1}
  ]
 }
 Cubes {
  emitterUpdateScript [
   &EmitterState(Life Cycle Mode=::NewEnumerator1, Loop Behavior=::NewEnumerator1, Loop Duration=1.5, UseLoopDelay=false)
   &SpawnBurst_Instantaneous(
    Spawn Count=&RandomRangeInt(Minimum=7, Maximum=10)
    Use Spawn Probability=false
    UseLoopCountLimit=false
   )
   &SpawnBurst_Instantaneous(
    Spawn Count=&RandomRangeInt(Minimum=2, Maximum=3)
    Spawn Group=1
    Use Spawn Probability=false
    UseLoopCountLimit=false
   ) {module=SpawnBurst_Instantaneous001}
  ]
  particleSpawnScript [
   &InitializeParticle(
    Color=<rgba 0.0218 0.5439 1 1>
    Lifetime=&UniformRangedFloat(Minimum=5, Maximum=5)
    Mesh Scale Mode=::NewEnumerator3
    Mesh Uniform Scale=&MergeHelper_FloatSwitch(
     0=&RandomRangeFloat(Minimum=0.15, Maximum=0.4)
     1=&RandomRangeFloat(Minimum=0.6, Maximum=0.9)
     2=?
     3=?
     Index=$Emitter.SpawnGroup
     Num=::NewEnumerator0
    )
    UsePositionOffset=false
    Write Material Random=true
   ) {module=InitializeParticle001}
   &CameraOffset(Camera Offset Amount=5)
   &SocketLocation(SamplePosOffset=true, Socket Reader=$System.Socket Reader, Socket Sampling Mode=::NewEnumerator3, VelocitySpeedLimit=false)
   &ShapeLocation(
    Cone Angle=60
    Cone Length=12
    MaskSpawnGroup=false
    Offset Mode=::NewEnumerator0
    Override Seed=false
    Shape Primitive=::NewEnumerator5
    UseOwnerScale=true
   ) {module=ShapeLocation001}
   &InitialMeshOrientation(Use Rotation Vector=false)
   &DynamicMaterialParameters(
    Index 0 Param 1=&FloatFromCurve(
     FloatCurve=DataInterface() {propertyValues {Curve {
        defaultValue=1
        keys [
         {arriveTangent=0.0618, interpMode=RCIM_Cubic, leaveTangent=0.0618, tangentMode=RCTM_User, time=0, value=0}
         <at 0.3 0 RCIM_Cubic>
         {interpMode=RCIM_Cubic, tangentMode=RCTM_User, time=1, value=0}
        ]
       }}}
     Scale Curve=0.5
    )
    Index 0 Param 2=1
    Index 0 Param 3=&RandomRangeFloat(Minimum=0.75, Maximum=1.75)
    Index 0 Param 4=1
    Index 1 Param 1=&FN_FrontEndFloatModifier(Gameplay=1, Front End=1)
    Module.Param0WriteEnabled=true
    Module.Param0WriteEnabled1=true
    Module.Param1WriteEnabled=true
    Module.Param1WriteEnabled1=true
    Module.Param2WriteEnabled=true
    Module.Param2WriteEnabled1=true
    Module.Param3WriteEnabled=true
    Module.Param3WriteEnabled1=true
    Write Parameter Index 1=true
   )
  ]
  particleUpdateScript [
   &ParticleState(Kill Particles When Lifetime Has Elapsed=false)
   &ScaleMeshSize(Scale Factor=&VectorFromFloat(Value=&FloatFromCurve(
      FloatCurve=DataInterface() {propertyValues {Curve {keys [<at 0 0 RCIM_Cubic>,<at 0.5 1 RCIM_Cubic>]}}}
      CurveIndex=&Multiply_Float(A=$System.Age, B=&RandomRangeFloat(Minimum=0.5, Maximum=1.5))
      Scale Curve=&OneMinusFloat(Float=$User.Reactivity)
     )))
   &UpdateMeshOrientation(
    Rotation Rate=&RandomRangeFloat(Minimum=0, Maximum=0.2)
    Rotation Vector=&RandomRangeVector(Minimum=<-1 -1 -1>, Maximum=<1 1 1>)
   )
   &Drag(Drag=&UniformRangedFloat(Minimum=4, Maximum=8), Use Linear Drag=true, Use Rotational Drag=false)
   &SolveForcesAndVelocity(Clamp Velocity=false, Limit Acceleration=false)
   &ScaleColor(ScaleA=true, ScaleRGB=true)
   /SetVariables(Particles.Position=&AddVectorToPosition(
     Position=$Particles.Initial.Position
     Vector=&Multiply_Vector(
      A=&MakeVector(
       X=&Waveform(
        Global Input Position=&RandomRangeFloat(Minimum=0, Maximum=10)
        "Waveform [1]"=::NewEnumerator12
        "[1] Frequency"=&RandomRangeFloat(Minimum=0.1, Maximum=0.45)
        "[1] Amplitude Min/Max"=<-1 1>
       )
       Y=&Waveform(
        Global Input Position=&RandomRangeFloat(Minimum=0, Maximum=10)
        "Waveform [1]"=::NewEnumerator12
        "[1] Frequency"=&RandomRangeFloat(Minimum=0.1, Maximum=0.45)
        "[1] Amplitude Min/Max"=<-1 1>
       )
       Z=&Waveform(
        Global Input Position=&RandomRangeFloat(Minimum=0, Maximum=10)
        "Waveform [1]"=::NewEnumerator12
        "[1] Frequency"=&RandomRangeFloat(Minimum=0.1, Maximum=0.45)
        "[1] Amplitude Min/Max"=<-1 1>
       )
      )
      B=&RandomRangeVector(Minimum=<1 1 1>, Maximum=<5 5 5>)
     )
    )) {isSetParameters=true}
  ]
  rendererClasses [&NiagaraMeshRendererProperties]
  renderers [{rendererClass=&NiagaraMeshRendererProperties, rendererIndex=0}]
 }
 Headlights {
  emitterUpdateScript [
   &EmitterState(
    Inactive Response=::NewEnumerator1
    Life Cycle Mode=::NewEnumerator1
    Loop Behavior=::NewEnumerator1
    Loop Duration Mode=::NewEnumerator1
    UseLoopDelay=false
   )
   &SpawnBurst_Instantaneous(Spawn Count=2, Use Spawn Probability=false, UseLoopCountLimit=false)
   &SpawnBurst_Instantaneous(Spawn Count=2, Spawn Time=0.2, Use Spawn Probability=false, UseLoopCountLimit=false) {module=SpawnBurst_Instantaneous001}
  ]
  particleSpawnScript [
   &FN_SetSpawnGroup(Number of Spawn Groups=2)
   /SetVariables(Particles.MirrorAxis:Vector3f=&MaskVectorBySpawnGroup(Masked Vector=<-1 1 1>, Passthrough Vector=<1 1 1>)) {isSetParameters=true}
   &InitializeParticle(
    Color=<rgba 0.7083 0.5463 0.3468 0.15>
    Mesh Scale Mode=::NewEnumerator3
    Sprite Size Mode=::NewEnumerator3
    Uniform Sprite Size=8
    UsePositionOffset=true
    Write Material Random=true
   )
   &CameraOffset(Camera Offset Amount=5)
   &DynamicMaterialParameters(
    Index 0 Param 1=&Multiply_Float(A=-19, B=&MakeFloatFromVector(Vector=<0 0 0>, Channel=::NewEnumerator0))
    Index 0 Param 2=-1
    Module.Param0WriteEnabled=true
    Module.Param1WriteEnabled=true
    Module.Param2WriteEnabled=false
    Module.Param3WriteEnabled=false
   )
  ]
  particleUpdateScript [
   &ParticleState(Kill Particles When Lifetime Has Elapsed=false)
   &SocketLocation(
    Orientation Sampling=::NewEnumerator1
    SamplePosOffset=true
    Sampled Position Offset=<-1 0 1>
    Socket Index=&Add_Integer(A=4, B=$Particles.SpawnGroup)
    Socket Reader=$System.Socket Reader
    Socket Sampling Mode=::NewEnumerator3
    VelocitySpeedLimit=false
   )
   &ScaleColor(
    Scale Alpha=&Multiply_Float(
     A=&Multiply_Float(A=$System.FadeIn, B=&OneMinusFloat(Float=$User.Reactivity))
     B=&FN_FrontEndFloatModifier(
      Gameplay=&FNTimeOfDayScalar(Night Scale=1, Morning Scale=0, Day Scale=0, Evening Scale=0.7)
      Front End=1
     )
    )
    ScaleA=true
    ScaleRGB=true
   )
   /SetVariables(
    Particles.FlareAlignment:Vector3f=<0 0 1>
    Particles.FlareCameraOffset=10
    Particles.FlareColor=&MultiplyLinearColor(Linear Color=$Particles.Color, Scale Factor=&MakeVector4_FromFloat(Float=2))
    Particles.FlareFacing:Vector3f=<1 0 0>
    Particles.FlarePosition:NiagaraPosition=&AddVectorToPosition(Position=$Particles.Position, Vector=<1 0 -1>)
    Particles.FlareSize:Vector2f=&Multiply_Vector2D_ByFloat(Float=0.15, Vector2D=<100 50>)
   ) {isSetParameters=true}
   /SetVariables(
    Particles.GlowCameraOffset:NiagaraFloat=&FN_FrontEndFloatModifier(Gameplay=1, Front End=0)
    Particles.GlowColor=&MultiplyLinearColor(
     Linear Color=$Particles.Color
     Scale Factor=&MakeVector4_FromFloat(Float=&Multiply_Float(
       A=&Multiply_Float(A=8, B=$System.FadeIn)
       B=&ClampFloat(Float=&OneMinusFloat(Float=&Multiply_Float(
          A=0.7
          B=&VectorLength(
           Vector=/Niagara/DynamicInputs/Vectors/CrossProduct(
            Vector A:Vector3f=&ReturnCameraVector(Camera Vectors=::NewEnumerator1, Coordinate Space=::Simulation)
            Vector B:Vector3f=<0 -1 0>
            Normalize Incoming Vectors:NiagaraBool=true
           )
           Fallback Vector=<0 -1 0>
          )
         )))
      ))
    )
    Particles.GlowPosition:NiagaraPosition=&AddVectorToPosition(Position=$Particles.Position, Vector=<1 0 -1>)
    Particles.GlowSize=&Vector2DFromFloat(Value=&FN_FrontEndFloatModifier(Gameplay=1, Front End=0))
    Particles.glowalignement:Vector3f=<0 0 1>
    Particles.glowfacing:Vector3f=<1 0 0>
   ) {isSetParameters=true}
   /SetVariables(
    Particles.BigGlowColor:LinearColor=&MultiplyLinearColor(
     Linear Color=$Particles.Color
     Scale Factor=&MakeVector4_FromFloat(Float=&Multiply_Float(
       A=&Multiply_Float(A=1, B=$System.FadeIn)
       B=&ClampFloat(Float=&OneMinusFloat(Float=&Multiply_Float(
          A=0.4
          B=&VectorLength(
           Vector=&CrossProduct(
            Vector A:Vector3f=&ReturnCameraVector(Camera Vectors=::NewEnumerator1, Coordinate Space=::Simulation)
            Vector B:Vector3f=<0 -1 0>
            Normalize Incoming Vectors:NiagaraBool=true
           )
           Fallback Vector=<0 -1 0>
          )
         )))
      ))
    )
    Particles.BigGlowSize:Vector2f=&Vector2DFromFloat(Value=8)
   ) {isSetParameters=true}
   &SpriteFacingAndAlignment(Alignment Coordinate Space=::Local, Sprite Alignment=<-1 0 0>, WriteAlignment=true, WriteFacing=false) {module=SpriteFacingAndAlignment001}
   &ScaleColor(Scale Alpha=$System.FadeIn, ScaleA=true, ScaleRGB=true) {module=ScaleColor002}
  ]
  rendererClasses [&NiagaraSpriteRendererProperties]
  renderers [
   {rendererClass=&NiagaraSpriteRendererProperties}
   {rendererClass=&NiagaraSpriteRendererProperties, rendererIndex=1}
   {rendererClass=&NiagaraSpriteRendererProperties, rendererIndex=2}
   {rendererClass=&NiagaraSpriteRendererProperties, rendererIndex=3}
  ]
 }
 Illumination {
  emitterUpdateScript [
   &EmitterState()
   &SpawnBurst_Instantaneous(Use Spawn Probability=false, UseLoopCountLimit=true)
  ]
  particleSpawnScript [
   &InitializeParticle(
    Color=<rgba 0.3 1.2046 2 0.4>
    Lifetime=2.5
    Mesh Scale=<0.25 0.25 1>
    Mesh Scale Mode=::NewEnumerator4
    Position Mode=::NewEnumerator1
    Sprite Size Mode=::NewEnumerator3
    Uniform Sprite Size=20
    UsePositionOffset=false
    Write Material Random=true
   )
   &DynamicMaterialParameters(
    Index 0 Param 1=1
    Index 0 Param 2=0.7
    Module.Index0Float0Write=true
    Module.Index0Float1Write=true
    Module.Index0Float2Write=true
    Module.Index0Float3Write=true
   )
   /SetVariables(
    Particles.LightConePosition:NiagaraPosition=&AddVectorToPosition(Position=$Engine.Emitter.SimulationPosition, Vector=<0 0 10>)
    Particles.LightPosition:NiagaraPosition=&AddVectorToPosition(Position=$Engine.Emitter.SimulationPosition, Vector=<0 -10 22>)
   ) {isSetParameters=true}
   &SocketLocation(SamplePosOffset=false, Socket Reader=$System.Socket Reader, Socket Sampling Mode=::NewEnumerator3, VelocitySpeedLimit=false)
   /SetVariables(
    Particles.BaseAlignment:Vector3f=<0 1 0>
    Particles.BaseColor:LinearColor=<rgba 1.2213 3.0848 8 1>
    Particles.BaseDynamic:Vector4f=<0.85 0.2 1 1>
    Particles.BaseFacing:Vector3f=<0 0 1>
    Particles.BasePosition:NiagaraPosition=&AddVectorToPosition(Position=$Particles.Position, Vector=<0 0 0.6>)
    Particles.BaseSize:Vector2f=&Vector2DFromFloat(Value=8)
   ) {isSetParameters=true}
  ]
  particleUpdateScript [
   &ParticleState(Kill Particles When Lifetime Has Elapsed=false, Let Infinitely Lived Particles Die When Emitter Deactivates=true)
   &CameraOffset(Camera Offset Amount=5)
   &ScaleColor(
    Scale Alpha=&OneMinusFloat(Float=$User.Reactivity)
    Scale RGB=&VectorFromFloat(Value=&Lerp_Float(
      A=1
      B=&Waveform(Global Amplitude Scale=0.5, "Waveform [1]"=::NewEnumerator8, "[1] Frequency"=15, "[1] Amplitude Min/Max"=<0.5 1>, "[1] Bias"=0.7)
      Alpha=0.4
     ))
    ScaleA=true
    ScaleRGB=true
   )
   &FN_Light_Attributes(
    Color=&MultiplyLinearColor(Linear Color=<rgba 0 0.531 1 1>, Scale Factor=&MakeVector4_FromLinearColor(LinearColor=$Particles.Color))
    Create Color Attribute=true
    Create Exponent Attribute=true
    Create Radius Attribute=true
    Enable Particle Light Enable=true
    Exponent=1.5
    Light Enabled=true
    Radius=25
    Scale Color=&Multiply_Float(
     A=&Add_Float(A=5, B=&Multiply_Float(A=20, B=$User.GlitchSpawn))
     B=&OneMinusFloat(Float=$User.Reactivity)
    )
   )
   /SetVariables(Particles.BaseColor:LinearColor=&MultiplyLinearColor(
     Linear Color=$Particles.Initial.BaseColor
     Scale Factor=&MakeVector4FromVectorAndFloat("Vector (RGB)"=<1 1 1>, "Float (Alpha)"=&OneMinusFloat(Float=$User.Reactivity))
    )) {isSetParameters=true}
  ]
  rendererClasses [
   &NiagaraSpriteRendererProperties
   &NiagaraLightRendererProperties
   &NiagaraMeshRendererProperties
  ]
  renderers [
   {rendererClass=&NiagaraSpriteRendererProperties}
   {rendererClass=&NiagaraLightRendererProperties, rendererIndex=1}
   {rendererClass=&NiagaraMeshRendererProperties, rendererIndex=2}
   {rendererClass=&NiagaraSpriteRendererProperties, rendererIndex=3}
  ]
 }
 ScreenLights {
  emitterUpdateScript [
   &EmitterState(Life Cycle Mode=::NewEnumerator1, Loop Behavior=::NewEnumerator1, Loop Duration=9.4167, UseLoopDelay=false)
   &SpawnBurst_Instantaneous(Spawn Count=3, Use Spawn Probability=false, UseLoopCountLimit=false)
  ]
  particleSpawnScript [
   &PartitionParticles(Number Of Partitions=3)
   &InitializeParticle(
    Color=<rgba 1 1 1 0.3>
    Lifetime=2
    Mesh Scale Mode=::NewEnumerator3
    Mesh Uniform Scale=0.6
    Sprite Size=&Multiply_Vector2D(A=<14 14>, B=<1 1>)
    Sprite Size Mode=::NewEnumerator4
    UsePositionOffset=false
    Write Material Random=true
   )
   &CameraOffset(Camera Offset Amount=5)
   &SocketLocation(
    SamplePosOffset=true
    Sampled Position Offset=<0 1 0.5>
    Socket Index=&Add_Integer(A=$Particles.PartitionParticles.Partition, B=1)
    Socket Reader=$System.Socket Reader
    Socket Sampling Mode=::NewEnumerator3
    VelocitySpeedLimit=false
   )
   /SetVariables(
    Particles.ScreenAlignment:Vector3f=&ReturnMeshOrientationAxis(Mesh Orientation Axis=::ZAxis, Mesh Orientation Quat=$Particles.MeshOrientation)
    Particles.ScreenFacing:Vector3f=&ReturnMeshOrientationAxis(Mesh Orientation Axis=::YAxis, Mesh Orientation Quat=$Particles.MeshOrientation)
   ) {isSetParameters=true}
  ]
  particleUpdateScript [
   &ParticleState(Kill Particles When Lifetime Has Elapsed=false)
   &ScaleColor(
    Scale Alpha=&Multiply_Float(
     A=&OneMinusFloat(Float=$User.Reactivity)
     B=&Lerp_Float(
      A=0.5
      B=2
      Alpha=&Abs_Float(Float:NiagaraFloat=&DotProduct(
        Vector A=&ReturnCameraVector(Camera Vectors=::NewEnumerator3, Coordinate Space=::Simulation)
        Vector B=/Game/Effects/Niagara/DynamicInputs/Engine/FN_GetSystemAxis(Axis:ENiagaraOrientationAxis=::XAxis, Flip Axis:NiagaraBool=false)
       ))
     )
    )
    ScaleA=true
    ScaleRGB=true
   )
   /SetVariables(Particles.Position=&AddVectorToPosition(
     Position=$Particles.Initial.Position
     Vector=&Lerp_Vector(
      A=&RotateVectorByQuaternion(Vector To Rotate=<0 -5 0>, Quaternion=$Particles.MeshOrientation)
      B=<0 0 0>
      Alpha=&ClampFloat(Float=&DotProduct(
        Vector A=&ReturnCameraVector(Camera Vectors=::NewEnumerator1, Coordinate Space=::Simulation)
        Vector B=&RotateVectorByQuaternion(Vector To Rotate=<0 -1 0>, Quaternion=$Particles.MeshOrientation)
       ))
     )
    )) {isSetParameters=true}
  ]
  rendererClasses [&NiagaraSpriteRendererProperties]
  renderers [{rendererClass=&NiagaraSpriteRendererProperties}]
 }
 Streaks {
  simTarget=GPUComputeSim
  emitterUpdateScript [
   &EmitterState(Life Cycle Mode=::NewEnumerator1, Loop Duration=1, Scalability Mode=::NewEnumerator1, UseLoopDelay=false)
   &SpawnRate(SpawnRate=30, Use Spawn Probability=false) {module=SpawnRate001}
  ]
  particleSpawnScript [
   /Niagara/Modules/Spawn/Initialization/InitializeParticle(
    Color=<rgba 1.2213 3.0848 8 0.2>
    Lifetime=&UniformRangedFloat(Minimum=0.2, Maximum=0.4)
    Mesh Scale=&Multiply_Vector(
     A=<1 1 1>
     B=&VectorFromFloat(Value=&UniformRangedFloat(Minimum=0.2, Maximum=1))
    )
    Module.Write Color=true
    Module.Write Lifetime=true
    Module.Write Mass=true
    Module.Write Material Random=true
    Module.Write Position=false
    Module.Write Scale=false
    Module.Write SpriteRotation=true
    Module.Write SpriteSize=true
    Position=&SimulationPosition()
    Sprite Rotation=90
    Sprite Size=&MakeVector2D(
     X=&UniformRangedFloat(Minimum=0.5, Maximum=1.5)
     Y=&UniformRangedFloat(Minimum=50, Maximum=100)
    )
   ) {enabled=false}
   /Niagara/Modules/Spawn/Initialization/V2/InitializeParticle(
    Color=<rgba 0.0763 0.1928 0.5 0.3>
    Lifetime=&UniformRangedFloat(Minimum=0.2, Maximum=0.4)
    Sprite Size=&MakeVector2D(
     X=&UniformRangedFloat(Minimum=0.5, Maximum=1.5)
     Y=&UniformRangedFloat(Minimum=40, Maximum=60)
    )
    Sprite Size Mode=::NewEnumerator4
    UsePositionOffset=false
    Write Material Random=true
   ) {module=InitializeParticle001}
   &ShapeLocation(
    Cone Angle=90
    Cone Length=2
    Cone Surface Distribution=0.5
    MaskSpawnGroup=false
    Offset Mode=::NewEnumerator0
    Override Seed=false
    Shape Primitive=::NewEnumerator5
    UseOwnerScale=true
   )
   &AddVelocityFromPoint(Module.Use Falloff Distance=false, Velocity Strength=5)
   /SetVariables(Particles.RandAlpha:NiagaraFloat=&RandomRangeFloat(Minimum=0.3, Maximum=1)) {isSetParameters=true}
   &SocketLocation(SamplePosOffset=true, Socket Reader=$System.Socket Reader, Socket Sampling Mode=::NewEnumerator3, VelocitySpeedLimit=false)
  ]
  particleUpdateScript [
   &ParticleState()
   &ScaleSpriteSize(Scale Sprite Size Mode=::NewEnumerator0, Uniform Scale Factor=0.4)
   &SolveForcesAndVelocity(Clamp Velocity=false, Limit Acceleration=false)
   &ScaleColor(
    Scale Alpha=&FloatFromCurve(
     FloatCurve=DataInterface() {propertyValues {Curve {keys [<at 0 0 RCIM_Cubic>,<at 0.5 1.1 RCIM_Cubic>,<at 1 0 RCIM_Cubic>]}}}
     Scale Curve=&Multiply_Float(A=$Particles.RandAlpha, B=&OneMinusFloat(Float=$User.Reactivity))
    )
    ScaleA=true
    ScaleRGB=true
   )
  ]
  rendererClasses [&NiagaraSpriteRendererProperties]
  renderers [{rendererClass=&NiagaraSpriteRendererProperties}]
 }
}
summary {
 systemName=NS_Backpack_DoubleConk_Idle
 emitters [
  {
   emitterName=Illumination
   rendererClasses [
    &NiagaraSpriteRendererProperties
    &NiagaraLightRendererProperties
    &NiagaraMeshRendererProperties
   ]
  }
  {emitterName=ConeGlow, rendererClasses [&NiagaraSpriteRendererProperties, &NiagaraMeshRendererProperties]}
  {emitterName=Streaks, simTarget=GPUComputeSim, rendererClasses [&NiagaraSpriteRendererProperties]}
  {emitterName=Cubes, rendererClasses [&NiagaraMeshRendererProperties]}
  {emitterName=ScreenLights, rendererClasses [&NiagaraSpriteRendererProperties]}
  {emitterName=Headlights, rendererClasses [&NiagaraSpriteRendererProperties]}
 ]
 userVariables [
  User.bIsFrontEnd {type=NiagaraBool}
  User.bIsFrontEndPreview {type=NiagaraBool}
  User.ForwardDot {type=NiagaraFloat, defaultValue=0, description=""}
  User.GlitchSpawn {type=NiagaraFloat, defaultValue=0, description=""}
  User.ReactiveTriggered {type=NiagaraBool}
  User.Reactivity {type=NiagaraFloat, defaultValue=0, description=""}
 ]
}
