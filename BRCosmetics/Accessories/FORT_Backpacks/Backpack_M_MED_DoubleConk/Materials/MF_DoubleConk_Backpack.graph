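# Material function that adds animated screens to the DoubleConk backpack's material attributes, with a static switch to bypass the effect.
# A ScreenSwitchState parameter blends between an "on" look, with tinted grid lights and logo, and an "off" look with sine-pulsed grid noise in base color.
# The Reboot icon is projected from local position onto side and back screens, glitched by time-stepped sphere masks that shift horizontal bands of its UVs.
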MaterialFunction /BRCosmetics/Accessories/FORT_Backpacks/Backpack_M_MED_DoubleConk/Materials/MF_DoubleConk_Backpack
graph {nodes [
  n1 = Constant() {r=0.3}
  Add(
   A=ConstantBiasScale(None=MaterialFunctionCall(Speed=@n1) {materialFunction=&DebugTimeSine}.Result) {bias=-0.5, scale=5}
   B=Multiply(A=NamedRerouteUsage())
  )
  n2 = VectorParameter() {group=DoubleConkBP, parameterName=DCBP_GridScale, defaultValue=<rgba 0.47 0.47 0 1>}
  n3 = Multiply(A=TextureCoordinate() {coordinateIndex=1}, B=AppendVector(A=@n2.R, B=@n2.R))
  BumpOffset(Coordinate=@n3, Height=ScalarParameter() {defaultValue=0.15, group=DoubleConkBP, parameterName=DCBP_BumpOffsetHeight})
  n4 = GetMaterialAttributes(None=FunctionInput() {inputType=FunctionInput_MaterialAttributes}) {attributeGetTypes [#1, #2]}
  n5 = NamedRerouteUsage()
  n6 = NamedRerouteUsage()
  n7 = NamedRerouteUsage()
  n8 = BlendMaterialAttributes(
   A=NamedRerouteDeclaration(None=SetMaterialAttributes!(MaterialAttributes=@n4.MaterialAttributes, BaseColor=@n5, EmissiveColor=@n6) {attributeSetTypes [#2, #1]}) {name=Mat_Out_ScreenOn, nodeColor=<rgba 1 0 0.0703 1>}
   B=NamedRerouteDeclaration(None=SetMaterialAttributes(MaterialAttributes=@n4.MaterialAttributes, BaseColor=@n7) {attributeSetTypes [#2, #1]}) {name=Mat_Out_ScreenOff, nodeColor=<rgba 1 0 0.0703 1>}
   Alpha=Saturate(None=ScalarParameter() {group=DoubleConkBP, parameterName=DCBP_ScreenSwitchState})
  )
  n9 = NamedRerouteUsage()
  FunctionOutput(None=StaticSwitchParameter(True=BlendMaterialAttributes(A=@n4.MaterialAttributes, B=@n8, Alpha=@n9), False=@n4.MaterialAttributes) {defaultValue=true, parameterName=UseMaterialScreenAnimation})
  NamedRerouteDeclaration(None=ComponentMask(None=FunctionInput() {inputName=ScreenMask}) {r=true}) {name=ScreenMask, nodeColor=<rgba 0 1 0.7422 1>}
  n10 = TextureSample(UVs=@n3) {texture=/BRCosmetics/Weapons/FORT_Melee/Pickaxe_DoubleConk/Textures/T_GridNoise}
  n11 = Frac(None=Multiply(A=Time()) {constB=0.4})
  n12 = ConstantBiasScale(None=Sine(None=Add(A=@n10.R, B=@n11)))
  n13 = LinearInterpolate(
   A=ScalarParameter() {defaultValue=0.3, group=DoubleConkBP, parameterName=DCBP_ScreenGridEmissive_Low}
   B=ScalarParameter() {defaultValue=1, group=DoubleConkBP, parameterName=DCBP_ScreenGridEmissive_High}
   Alpha=@n12
  )
  n14 = NamedRerouteUsage()
  NamedRerouteDeclaration(None=NamedRerouteDeclaration(None=Multiply(A=@n13, B=@n14)) {name=Screen_Off, nodeColor=<rgba 0 1 0.9766 1>}) {name=BaseColor_Out_Off, nodeColor=<rgba 1 0 0.5391 1>}
  NamedRerouteDeclaration(None=@n12) {name=GridNoise, nodeColor=<rgba 0 0.4609 1 1>}
  NamedRerouteDeclaration(None=@n4.EmissiveColor) {name=EmissiveIn, nodeColor=<rgba 1 0 0.1406 1>}
  NamedRerouteDeclaration(None=@n4.BaseColor) {name=BaseColorIn, nodeColor=<rgba 0 1 0.5781 1>}
  n15 = LinearInterpolate(A=NamedRerouteUsage(), Alpha=NamedRerouteUsage()) {constB=0}
  n16 = TextureCoordinate() {coordinateIndex=1}
  n17 = Constant() {r=50}
  n18 = Multiply(
   A=TextureSample(UVs=Multiply(A=@n16, B=@n17)) {texture=/BRCosmetics/Weapons/FORT_Melee/Pickaxe_DoubleConk/Textures/T_GridDebug}.RGB
   B=VectorParameter() {group=DoubleConkBP, parameterName=DCBP_GridLightColor, defaultValue=<rgba 0 5 10 1>}.RGB
  )
  NamedRerouteDeclaration(None=LinearInterpolate(
    A=Add(A=@n15, B=@n18)
    B=VectorParameter() {parameterName=DCBP_LogoEmissiveColor, defaultValue=<rgba 0.0269 0.1325 0.2812 1>}.RGB
    Alpha=NamedRerouteUsage()
   )) {name=Emissive_Out_On, nodeColor=<rgba 0 0.0625 1 1>}
  n19 = MaterialFunctionCall(Float3=LocalPosition().XYZ) {materialFunction=&BreakOutFloat3Components}
  n20 = Step(Y=@n19.G) {constX=-5}
  n21 = MaterialFunctionCall(X=LinearInterpolate(A=Multiply(A=@n19.R) {constB=-1}, B=@n19.R, Alpha=@n20), Y=@n19.B) {materialFunction=&MakeFloat2}
  n22 = VectorParameter() {group=DoubleConkBP, parameterName=SideScreenLocationOffset2, defaultValue=<rgba 0.9 0.77 0 1>}
  n23 = AppendVector(A=@n22.R, B=@n22.R)
  n24 = VectorParameter() {group=DoubleConkBP, parameterName=SideScreenLocationOffset, defaultValue=<rgba 0.05 0.77 0 1>}
  n25 = AppendVector(A=@n24.R, B=@n24.R)
  n26 = MaterialFunctionCall(Float2=Add(A=Multiply(A=@n21.Result) {constB=-0.21}, B=LinearInterpolate(A=@n23, B=@n25, Alpha=@n20))) {materialFunction=&BreakOutFloat2Components}
  n27 = Subtract(A=@n26.R) {constB=0.1}
  n28 = Divide(A=Floor(None=Multiply(A=@n26.G) {constB=20})) {constB=20}
  n29 = Multiply(A=Time()) {constB=6}
  n30 = Sine(None=Divide(A=Floor(None=@n29)) {constB=6})
  n31 = Constant() {r=44.26}
  n32 = Frac(None=Add(A=Multiply(A=@n30, B=@n31)) {constB=1.985})
  n33 = LinearInterpolate(A=@n26.R, B=@n27, Alpha=SphereMask(A=@n28, B=Multiply(A=@n32) {constB=2}) {attenuationRadius=0.15})
  n34 = Add(A=@n27) {constB=0.2}
  n35 = Time()
  n36 = Divide(A=Floor(None=Multiply(A=@n35) {constB=5})) {constB=5}
  n37 = Add(A=Multiply(A=Sine(None=@n36), B=Constant() {r=21.46})) {constB=49.5}
  n38 = SphereMask(A=@n28, B=Multiply(A=Frac(None=@n37)) {constB=1.2}) {attenuationRadius=0.1}
  n39 = TextureSampleParameter2D(UVs=AppendVector(A=LinearInterpolate(A=@n33, B=@n34, Alpha=@n38), B=@n26.G)) {
   group=DoubleConkBP
   parameterName=DCBP_RebootIcon
   samplerSource=SSM_Clamp_WorldGroupSettings
   texture=/BRCosmetics/Accessories/FORT_Backpacks/Backpack_M_MED_DoubleConk/Textures/Reboot_v02
  }
  n40 = MaterialFunctionCall(Float3=LocalPosition().XYZ) {materialFunction=&BreakOutFloat3Components}
  n41 = MaterialFunctionCall(X=@n40.G, Y=@n40.G) {materialFunction=&MakeFloat2}
  n42 = Multiply(A=Constant2Vector() {g=-1, r=1}) {constB=0.22}
  n43 = VectorParameter() {parameterName=BackScreenLocationOffset, defaultValue=<rgba 2.9 0.75 0 1>}
  n44 = MaterialFunctionCall(Float2=Add(A=Multiply(A=@n41.Result, B=@n42) {constB=-0.2}, B=AppendVector(A=@n43.R, B=@n43.R))) {materialFunction=&BreakOutFloat2Components}
  n45 = Subtract(A=@n44.R) {constB=0.1}
  n46 = Divide(A=Floor(None=Multiply(A=@n44.G) {constB=20})) {constB=20}
  n47 = Multiply(A=Time()) {constB=6}
  n48 = Sine(None=Divide(A=Floor(None=@n47)) {constB=6})
  n49 = Constant() {r=44.26}
  n50 = Frac(None=Add(A=Multiply(A=@n48, B=@n49)) {constB=1.985})
  n51 = LinearInterpolate(A=@n44.R, B=@n45, Alpha=SphereMask(A=@n46, B=Multiply(A=@n50) {constB=2}) {attenuationRadius=0.15})
  n52 = Add(A=@n45) {constB=0.2}
  n53 = Time()
  n54 = Divide(A=Floor(None=Multiply(A=@n53) {constB=5})) {constB=5}
  n55 = Add(A=Multiply(A=Sine(None=@n54), B=Constant() {r=21.46})) {constB=49.5}
  n56 = SphereMask(A=@n46, B=Multiply(A=Frac(None=@n55)) {constB=1.2}) {attenuationRadius=0.1}
  n57 = TextureSampleParameter2D(UVs=AppendVector(A=LinearInterpolate(A=@n51, B=@n52, Alpha=@n56), B=@n44.G)) {
   group=DoubleConkBP
   parameterName=DCBP_RebootIcon
   samplerSource=SSM_Clamp_WorldGroupSettings
   texture=&Reboot_v02
  }
  NamedRerouteDeclaration(None=LinearInterpolate(B=Add(A=@n39.A, B=@n57.A), Alpha=Step(Y=@n19.B) {constX=15})) {name=Logo, nodeColor=<rgba 0.8516 0 1 1>}
  NamedRerouteDeclaration(None=LinearInterpolate(
    A=NamedRerouteUsage()
    B=VectorParameter() {group=DoubleConkBP, parameterName=DCBP_LogoBaseColor, defaultValue=<rgba 0.0113 0.0475 0.0833 1>}.RGB
    Alpha=NamedRerouteUsage()
   )) {name=BaseColor_Out_On, nodeColor=<rgba 0 1 0.7656 1>}
  n58 = LocalPosition() {localOrigin=InstancePreSkinning}
  ShaderStageSwitch(PixelShader=VertexInterpolator(VS=@n58.XYZ).PS, VertexShader=@n58.XYZ)
 ]}
