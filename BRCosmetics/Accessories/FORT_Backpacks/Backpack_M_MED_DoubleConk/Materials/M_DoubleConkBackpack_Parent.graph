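# Parent material for the DoubleConk backpack cosmetic, built on the shared 2023 character material pipeline with material attributes.
# The character PreFX function feeds a backpack-specific function, MF_DoubleConk_Backpack, whose result passes into the character PostFX function.
# A ScreenMask texture parameter, sampled on the second UV channel, drives the backpack function.
# The shading model comes from the graph, and skeletal mesh, morph target and mesh deformer usage are enabled.
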Material /BRCosmetics/Accessories/FORT_Backpacks/Backpack_M_MED_DoubleConk/Materials/M_DoubleConkBackpack_Parent
graph {
 properties {
  bUseMaterialAttributes=true
  bUsedWithMeshDeformer=true
  bUsedWithMorphTargets=true
  bUsedWithSkeletalMesh=true
  pixelDepthOffsetMode=PDOM_Legacy
  shadingModel=MSM_FromMaterialExpression
  subsurfaceProfile=/Game/Packages/Fortress_SharedMaterials/SubSurfaceProfile/Character_SubsurfaceProfile
  usedShadingModels=Default Lit | Subsurface | Clear Coat | Subsurface Profile | Cloth
 }
 nodes [
  n1 = MaterialFunctionCall() {materialFunction=&MF_CharacterParent_2023_PreFX}
  n2 = TextureSampleParameter2D(UVs=TextureCoordinate() {coordinateIndex=1}) {
   group=DoubleConkBP
   parameterName=ScreenMask
   texture=/BRCosmetics/Accessories/FORT_Backpacks/Backpack_M_MED_DoubleConk/Textures/T_DoubleConk_Backpack_Mask
  }
  MaterialFunctionCall(
   Material=MaterialFunctionCall!(In=@n1.Result, ScreenMask=@n2.RGB) {materialFunction=&/MF_DoubleConk_Backpack}.Result
   M=@n1.M
   Pre Skinned Local Position=@n1.M
   Distance Blend=@n1.M
   Fresnel=@n1.M
  ) {materialFunction=&MF_CharacterParent_2023_PostFX}
 ]
}
